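# Physics asset for the M_MED_SatireCane_Ode player body cosmetic, set up for secondary motion of a belt-hung piece, likely a weapon.
# Three box bodies chain from a small sphere at mstr_belt_weapon, which joins the pelvis; the last box is much larger and elongated.
# Belt joints have narrow swing limits and little or no twist.
# Pelvis, spine_02 and left thigh capsules approximate the nearby body, linked by free-rotation constraints.
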PhysicsAsset /BRCosmetics/Characters/Player/Male/Medium/Bodies/M_MED_SatireCane_Ode/Meshes/M_MED_SatireCane_Ode_Physics
bodies {
 dyn_belt1 [{shapeType=Box, extentX=2.999, extentY=0.6942, extentZ=5.7579, shapeName=dyn_belt1_box, center=<2.0537 0 0>, rotation=<rot 0 0.0001 90>}]
 dyn_belt2 [{shapeType=Box, extentX=2.999, extentY=0.6942, extentZ=5.7579, shapeName=dyn_belt1_box, center=<2.4688 0 0>, rotation=<rot 0 0.0001 90>}]
 dyn_belt3 [{
   shapeType=Box
   extentX=9.0273
   extentY=4.2675
   extentZ=46.8127
   shapeName=dyn_belt3_box
   center=<2.4945 0.3061 3.0253>
   rotation=<rot 0 94.7161 -88.901>
  }]
 mstr_belt_weapon [{shapeType=Sphere, radius=0.505, shapeName=mstr_belt_weapon_sphere}]
 pelvis [{shapeType=Capsule, length=9.6277, radius=10.7567, shapeName=pelvis_capsule}]
 spine_02 [{shapeType=Capsule, length=7.66, radius=9.7045, shapeName=spine_02_capsule, center=<2.6449 1.4163 -0.0059>, rotation=<rot 0.5306 -90 0.0837>}]
 thigh_l [{shapeType=Capsule, length=24.1823, radius=8.0989, shapeName=thigh_l_capsule, center=<-12.1236 0 0>, rotation=<rot -90 -0.0001 0>}]
}
constraints [
 {
  bone1Name=thigh_l
  bone2Name=pelvis
  swing1LimitDegrees=45
  swing1Motion=Free
  swing2LimitDegrees=45
  swing2Motion=Free
  twistLimitDegrees=45
  twistMotion=Free
 }
 {bone1Name=dyn_belt1, bone2Name=mstr_belt_weapon, swing1LimitDegrees=15, swing2LimitDegrees=15, twistLimitDegrees=0}
 {bone1Name=dyn_belt2, bone2Name=dyn_belt1, swing1LimitDegrees=15, swing2LimitDegrees=15, twistLimitDegrees=0}
 {bone1Name=dyn_belt3, bone2Name=dyn_belt2, swing1LimitDegrees=15, swing2LimitDegrees=15, twistLimitDegrees=5}
 {
  bone1Name=mstr_belt_weapon
  bone2Name=pelvis
  swing1LimitDegrees=45
  swing1Motion=Free
  swing2LimitDegrees=45
  swing2Motion=Free
  twistLimitDegrees=45
  twistMotion=Free
 }
 {
  bone1Name=spine_02
  bone2Name=pelvis
  swing1LimitDegrees=45
  swing1Motion=Free
  swing2LimitDegrees=45
  swing2Motion=Free
  twistLimitDegrees=45
  twistMotion=Free
 }
]
