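# Post-process style animation blueprint for the SatireCane_Ode male medium body cosmetic, layering secondary motion onto the character pose.
# The parent mesh pose feeds a Control Rig, then RigidBody physics simulated relative to mstr_belt_weapon, then a bone modifier shrinking dyn_scabbard to near zero.
# The event graph polls the owner's montage during emotes, flagging when either SatireCane_Ode emote passes an early time threshold and clearing it on emote end.
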AnimBlueprint /BRCosmetics/Characters/Player/Male/Medium/Bodies/M_MED_SatireCane_Ode/Meshes/M_MED_SatireCane_Ode_AnimBP
target_skeleton=&/M_MED_SatireCane_Ode_Skeleton
graphs [
 AnimGraph {nodes [
   n1 = Control Rig(Source=Copy Pose From Mesh() {bUseAttachedParent=true}) {
    alphaBoolBlend {blendOption=Linear}
    controlRigAssetReference {blueprintRigClass=&/M_MED_SatireCane_Ode_CtrlRig.M_MED_SatireCane_Ode_CtrlRig_C}
   }
   n2 = RigidBody(ComponentPose=Local To Component(LocalPose=@n1)) {
    bOverrideWorldGravity=true
    overridePhysicsAsset=&/M_MED_SatireCane_Ode_Physics
    simulationSpace=BaseBoneSpace
    alphaBoolBlend {blendOption=Linear}
    baseBoneRef {boneName=mstr_belt_weapon}
    overrideWorldGravity {z=-980}
    simSpaceSettings {maxAngularAcceleration=30, maxAngularVelocity=10, maxLinearAcceleration=30, maxLinearVelocity=30}
   }
   n3 = "Transform (Modify) Bone - Bone: dyn_scabbard"!(ComponentPose=@n2, bAlphaBoolEnabled:bool=True, Translation="-2,0,0", Rotation="0,0,0", Scale="0.01,0.01,0.01") -> (Pose) {
    alphaInputType=Bool
    scaleMode=BMM_Replace
    scaleSpace=BCS_BoneSpace
    translationMode=BMM_Additive
    translationSpace=BCS_ParentBoneSpace
    alphaBoolBlend {blendOption=Linear}
    boneToModify {boneName=dyn_scabbard}
    scale=<0.01 0.01 0.01>
    translation {x=-2}
   }
   Output Pose(Result=Component To Local(ComponentPose=@n3)) {layerGroup=None}
  ]}
 EventGraph {nodes [
   Event BlueprintUpdateAnimation()
   > Branch(Condition=Get bIsPlayingEmote())
   n1 = Get AnimInst() -> (AnimInst:Anim Instance Object Reference)
   > Is Valid(InputObject=@n1)
   n2 = GetCurrentActiveMontage!(self:Anim Instance Object Reference=@n1) -> (ReturnValue:Anim Montage Object Reference)
   > Branch(Condition=OR Boolean(
     A="Equal (Object)"(A=@n2, B=/BRCosmetics/Animation/Game/MainPlayer/Emotes/SatireCane_Ode/CMM/Emote_SatireCane_Ode_CMM_M)
     B="Equal (Object)"(A=@n2, B=/BRCosmetics/Animation/Game/MainPlayer/Emotes/SatireCane_Ode/CMM/Pickaxe/Emote_SatireCane_Ode_CMM_Pickaxe_M)
    ))
   > Set EID_SatireCane_Ode played?(EID_SatireCane_Ode played?:bool="float > float"(A=Montage_GetPosition!(self:Anim Instance Object Reference=@n1, Montage:Anim Montage Object Reference=@n2) -> (ReturnValue:float), B=0.15)) -> (Output_Get:bool)
   Event OnEndPlayingEmote() {is=title K2Node_Event}
   > Branch(Condition=Get EID_SatireCane_Ode played?() -> (EID_SatireCane_Ode played?:bool))
   > Set EID_SatireCane_Ode played?(EID_SatireCane_Ode played?=false)
   n3 = Event OnCharacterCustomizationLoaded()
   Set AnimInst(execute=[@n3.then, @n4.then], AnimInst:Anim Instance Object Reference=GetAnimInstance(self=GetOwnerSkeletalMeshComponent())) -> (Output_Get:Anim Instance Object Reference)
   n4 = Event BlueprintInitializeAnimation()
  ]}
]
variables [
 AnimInst {type=AnimInstance, default=null}
 EID_SatireCane_Ode played? {type=boolean}
]
